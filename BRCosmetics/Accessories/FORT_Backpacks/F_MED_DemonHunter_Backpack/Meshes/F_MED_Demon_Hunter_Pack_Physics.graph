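# Physics asset for the Demon Hunter backpack cosmetic, giving its dangling pendant simulated collision and motion.
# A capsule body on backpack_root anchors a four-link pendant chain of three small capsules ending in a flat box.
# Each link is constrained to its parent with limited swing and tight twist, with the final box link restricted to a narrower swing.
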PhysicsAsset /BRCosmetics/Accessories/FORT_Backpacks/F_MED_DemonHunter_Backpack/Meshes/F_MED_Demon_Hunter_Pack_Physics
bodies {
 backpack_root [{shapeType=Capsule, length=3.2218, radius=11.9277, center=<-4.5696 -2.1819 0.0002>, rotation=<rot 0 0 90>}]
 dyn_pendant_01 [{shapeType=Capsule, length=1.2, radius=1.2, center=<1.2 0 0>, rotation=<rot 90 0 0>}]
 dyn_pendant_02 [{shapeType=Capsule, length=1, radius=0.5, center=<1 0 0>, rotation=<rot 90 0 0>}]
 dyn_pendant_03 [{shapeType=Capsule, length=1, radius=0.5, center=<1 0 0>, rotation=<rot 90 0 0>}]
 dyn_pendant_04 [{shapeType=Box, extentX=5, extentY=1, extentZ=8, center=<3 0 0>}]
}
constraints [
 {bone1Name=dyn_pendant_01, bone2Name=backpack_root, swing1LimitDegrees=45, swing2LimitDegrees=45, twistLimitDegrees=5}
 {bone1Name=dyn_pendant_02, bone2Name=dyn_pendant_01, swing1LimitDegrees=45, swing2LimitDegrees=45, twistLimitDegrees=10}
 {bone1Name=dyn_pendant_03, bone2Name=dyn_pendant_02, swing1LimitDegrees=45, swing2LimitDegrees=45, twistLimitDegrees=10}
 {bone1Name=dyn_pendant_04, bone2Name=dyn_pendant_03, swing1LimitDegrees=30, swing2LimitDegrees=30, twistLimitDegrees=10}
]
